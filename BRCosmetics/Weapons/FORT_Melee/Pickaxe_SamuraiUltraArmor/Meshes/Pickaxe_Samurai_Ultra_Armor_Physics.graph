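# Physics asset for the Samurai Ultra Armor pickaxe, rigging a bow and two hanging tassels for secondary motion.
# A capsule on the two-handed handle anchors a bow base, from which two bow loops and left and right tassel chains branch.
# Tassels use tight swing and twist limits along small sphere, box and capsule bodies.
# Inner bow joints swing freely with little twist, while outer bow segments are nearly rigid.
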PhysicsAsset /BRCosmetics/Weapons/FORT_Melee/Pickaxe_SamuraiUltraArmor/Meshes/Pickaxe_Samurai_Ultra_Armor_Physics
bodies {
 Melee_TwoHanded_Handle [{shapeType=Capsule, length=49.2331, radius=3.1205, center=<0.2095 -1.2688 53.9555>, rotation=<rot -0.9117 -89.9203 -0.8702>}]
 dyn_Bow01_l [{shapeType=Sphere, radius=1.5403}]
 dyn_Bow01_r [{shapeType=Sphere, radius=1.7326}]
 dyn_Bow02_l [{shapeType=Sphere, radius=2.2275, center=<0.8217 0 0>}]
 dyn_Bow02_r [{shapeType=Sphere, radius=2.0716}]
 dyn_bowBase [{shapeType=Sphere, radius=1.3019}]
 dyn_tassle01_l [{shapeType=Sphere, radius=1.5065, center=<0.4151 0 0>}]
 dyn_tassle01_r [{shapeType=Box, extentX=0.75, extentY=1.25, extentZ=1, center=<0.5912 0.0959 0>}]
 dyn_tassle02_l [{shapeType=Sphere, radius=1.5571, center=<1.2882 0.0864 0.1155>}]
 dyn_tassle02_r [{shapeType=Box, extentX=0.5, extentY=0.5, extentZ=0.5, center=<1.062 0 0>}]
 dyn_tassle03_l [{shapeType=Capsule, length=2.6676, radius=2.3393, center=<1.3835 0.2046 -0.2636>, rotation=<rot 0 88.831 -87.4164>}]
 dyn_tassle03_r [{shapeType=Box, extentX=1.2939, extentY=1.8518, extentZ=2.25, center=<0.816 0.3845 0>}]
 dyn_tassle04_r [{shapeType=Capsule, length=2.4457, radius=1.409, center=<0.8816 0 0.0138>, rotation=<rot 89.1015 -179.9951 -179.9952>}]
}
constraints [
 {bone1Name=dyn_bowBase, bone2Name=Melee_TwoHanded_Handle, swing1LimitDegrees=45, swing2LimitDegrees=45, twistLimitDegrees=45}
 {bone1Name=dyn_tassle01_l, bone2Name=dyn_bowBase, swing1LimitDegrees=20, swing2LimitDegrees=20, twistLimitDegrees=10}
 {bone1Name=dyn_tassle02_l, bone2Name=dyn_tassle01_l, swing1LimitDegrees=20, swing2LimitDegrees=20, twistLimitDegrees=5}
 {bone1Name=dyn_tassle03_l, bone2Name=dyn_tassle02_l, swing1LimitDegrees=20, swing2LimitDegrees=20, twistLimitDegrees=5}
 {bone1Name=dyn_tassle01_r, bone2Name=dyn_bowBase, swing1LimitDegrees=20, swing2LimitDegrees=20, twistLimitDegrees=10}
 {bone1Name=dyn_tassle02_r, bone2Name=dyn_tassle01_r, swing1LimitDegrees=20, swing2LimitDegrees=20, twistLimitDegrees=10}
 {bone1Name=dyn_tassle03_r, bone2Name=dyn_tassle02_r, swing1LimitDegrees=20, swing2LimitDegrees=20, twistLimitDegrees=10}
 {bone1Name=dyn_tassle04_r, bone2Name=dyn_tassle03_r, swing1LimitDegrees=20, swing2LimitDegrees=20, twistLimitDegrees=10}
 {
  bone1Name=dyn_Bow01_l
  bone2Name=dyn_bowBase
  swing1LimitDegrees=45
  swing1Motion=Free
  swing2LimitDegrees=45
  swing2Motion=Free
  twistLimitDegrees=2
 }
 {
  bone1Name=dyn_Bow01_r
  bone2Name=dyn_bowBase
  swing1LimitDegrees=45
  swing1Motion=Free
  swing2LimitDegrees=45
  swing2Motion=Free
  twistLimitDegrees=2
 }
 {bone1Name=dyn_Bow02_l, bone2Name=dyn_Bow01_l, swing1LimitDegrees=2, swing2LimitDegrees=2, twistLimitDegrees=2}
 {bone1Name=dyn_Bow02_r, bone2Name=dyn_Bow01_r, swing1LimitDegrees=2, swing2LimitDegrees=2, twistLimitDegrees=2}
]
