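# Animation Blueprint for the Samurai Ultra Armor pickaxe skeleton, adding simulated secondary motion to the weapon mesh.
# The input pose passes through a RigidBody node using the pickaxe's physics asset, with overridden world gravity and scaled, clamped component motion, then goes to the output.
# A spring controller for the dyn_Bow01_l bone and an unconfigured trail controller sit unconnected, and the event graph has no logic.
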AnimBlueprint /BRCosmetics/Weapons/FORT_Melee/Pickaxe_SamuraiUltraArmor/Meshes/Pickaxe_Samurai_Ultra_Armor_Skeleton_AnimBlueprint
target_skeleton=&/Pickaxe_Samurai_Ultra_Armor_Skeleton
graphs [
 AnimGraph {nodes [
   AnimGraphNode_SubInput_0 = Input Pose()
   n1 = RigidBody(ComponentPose=Local To Component(LocalPose=@AnimGraphNode_SubInput_0)) {
    bOverrideWorldGravity=true
    overridePhysicsAsset=&/Pickaxe_Samurai_Ultra_Armor_Physics
    alphaBoolBlend {blendOption=Linear}
    componentAppliedLinearAccClamp=<100 100 100>
    componentLinearAccScale=<0.25 0.25 0.25>
    componentLinearVelScale=<0.45 0.45 0.45>
    overrideWorldGravity {z=-980}
   }
   Output Pose(Result=Component To Local(ComponentPose=@n1)) {layerGroup=None}
   Spring controller - Bone: dyn_Bow01_l!(ComponentPose:Component Space Pose Link Structure, Alpha:float=1) -> (Pose:Component Space Pose Link Structure) {
    is=title AnimGraphNode_SpringBone
    bLimitDisplacement=true
    maxDisplacement=0.1
    springDamping=1
    springStiffness=5
    alphaBoolBlend {blendOption=Linear}
    springBone {boneName=dyn_Bow01_l}
   }
   Trail controller(Alpha=1) {rotationLimits [{limitMax=<180 180 180>}, {limitMax=<180 180 180>}], rotationOffsets [<0 0 0>, <0 0 0>], bLimitStretch=~, trailBone=~}
  ]}
 EventGraph {nodes [Event BlueprintUpdateAnimation()]}
]
variables []
